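# Niagara glitter effect for the SliceSprig medium male player body cosmetic, spawning sparse, short-lived spinning sprites around the torso.
# Its single Torso_Front emitter uses partition indices to pick hand-placed offsets, of which there are 14 for 12 partitions.
# Skeletal mesh sampling in particle update keeps sprites attached to the body.
# Curves animate size and alpha, with camera-distance remapping for size and time-of-day dimming for color.
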NiagaraSystem /BRCosmetics/Characters/Player/Male/Medium/Bodies/M_MED_SliceSprig/FX/NS_MED_SliceSprig_Glitter
dependencies {
 usedDataInterfaces [&NiagaraDataInterfaceArrayFloat3, &NiagaraDataInterfaceCurve, &NiagaraDataInterfaceSkeletalMesh]
 usedDynamicInputs []
 usedModules [
  &self:SystemSpawnScript.SystemScriptSource.SystemScriptGraph.NiagaraNodeAssignment_0.SetVariables_#1
  &SystemState
  &self:SystemSpawnScript.SystemScriptSource.SystemScriptGraph.NiagaraNodeAssignment_1.SetVariables_#2
 ]
}
emitters {Torso_Front {
  emitterUpdateScript [
   &EmitterState()
   &SpawnRate(SpawnRate=2.5, Use Spawn Probability=true)
  ]
  particleSpawnScript [
   &PartitionParticles(Number Of Partitions=12, Particles Per Partition=1, Partition Mode:ENiagara_PartitionMode=::NewEnumerator1)
   &InitializeParticle(
    Color Mode=::NewEnumerator0
    Lifetime=&RandomRangeFloat(Minimum=0.3, Maximum=0.4)
    Position Mode=::NewEnumerator0
    Sprite Rotation Mode=::NewEnumerator0
    Sprite Size Mode=::NewEnumerator1
    Uniform Sprite Size Max=$System.MaxSize
    Uniform Sprite Size Min=$System.MinSize
    Write Material Random=true
   )
  ]
  particleUpdateScript [
   &ParticleState()
   &SkeletalMeshLocation(
    Bone Sampling Mode=::NewEnumerator4
    MaskSpawnGroup=false
    Sampled Position Offset=&SelectVectorFromArray(
     Array Sampling Mode=::NewEnumerator0
     Direct Array Index=$Particles.Initial.PartitionParticles.Partition
     Vector Selection Array=DataInterface() {propertyValues {FloatData [
        <10 7.8302 14>
        <14.5 10.75 -10>
        <0 6 -13.5>
        <0 -5 13>
        <15.5 10 9.5>
        <19 2 10>
        <16 11.5 -2>
        <10 0 -16.5>
        <18 -12 2>
        <3 -5 -14.5>
        <15.5 -7 -12.5>
        <10 -14.25 8>
        <13 -14.75 -5>
        <4 -15 3>
       ]}}
     Random Seed=0
    )
    Skeletal Mesh=$System.Skeletal Mesh 1
   ) {module=SkeletalMeshLocation004}
   &SpriteRotationRate(Rotation Rate=250)
   &ScaleSpriteSize(
    Scale Sprite Size Mode=::NewEnumerator0
    Uniform Scale Factor=&FloatFromCurve(
     FloatCurve=DataInterface() {propertyValues {Curve {keys [{arriveTangent=2.9879, interpMode=RCIM_Cubic, leaveTangent=2.9879, tangentMode=RCTM_User, time=0, value=0}, <at 1 1 RCIM_Cubic>]}}}
     Scale Curve=&FN_LOD_Distance_Remap(Start Distance=100, End Distance=500, Minimum Output=0.7, Maximum Output=1.25, Clamp Output=true)
    )
   )
   &ScaleColor(
    Scale Alpha=&FloatFromCurve(FloatCurve=DataInterface() {propertyValues {Curve {keys [
         {arriveTangent=9.0725, interpMode=RCIM_Cubic, leaveTangent=9.0725, tangentMode=RCTM_User, time=0, value=0}
         {arriveTangent=-2.5249, interpMode=RCIM_Cubic, leaveTangent=-2.5249, tangentMode=RCTM_User, time=0.1408, value=0.9847}
         {arriveTangent=-0.0405, interpMode=RCIM_Cubic, leaveTangent=-0.0405, tangentMode=RCTM_User, time=1, value=0}
        ]}}})
    Scale RGB=&Multiply_VectorByFloat(
     Vector=<1 1 1>
     Float=&FNTimeOfDayScalar(Night Scale=0.2, Morning Scale=0.5, Day Scale=0.5, Evening Scale=0.5)
    )
    ScaleA=true
    ScaleRGB=true
   )
   &CameraOffset(Camera Offset Amount=2)
  ]
  rendererClasses [&NiagaraSpriteRendererProperties]
  renderers [{rendererClass=&NiagaraSpriteRendererProperties}]
 }}
summary {
 systemName=NS_MED_SliceSprig_Glitter
 emitters [{emitterName=Torso_Front, rendererClasses [&NiagaraSpriteRendererProperties]}]
 userVariables []
}
