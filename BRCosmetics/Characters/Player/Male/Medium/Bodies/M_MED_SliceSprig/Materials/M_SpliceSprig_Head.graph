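# Head material for the male medium SliceSprig player character, built on the shared Character Parent 2023 functions.
# The PreFX parent feeds a flipbook face function, whose result passes through a car clear coat function into the PostFX parent.
# The clear coat mask is the Tech Art mask multiplied by inverted face alpha, keeping the coat off the flipbook face.
# The clear coat function also drives a custom clear coat normal output.
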Material /BRCosmetics/Characters/Player/Male/Medium/Bodies/M_MED_SliceSprig/Materials/M_SpliceSprig_Head
graph {
 properties {
  bUseMaterialAttributes=true
  bUsedWithMeshDeformer=true
  bUsedWithMorphTargets=true
  bUsedWithNanite=true
  bUsedWithSkeletalMesh=true
  pixelDepthOffsetMode=PDOM_Legacy
  shadingModel=MSM_FromMaterialExpression
  usedShadingModels=Default Lit | Subsurface | Clear Coat | Subsurface Profile | Cloth
 }
 nodes [
  n1 = MaterialFunctionCall() {materialFunction=&MF_CharacterParent_2023_PreFX}
  n2 = MaterialFunctionCall!(Mouth UV, Eye UV, Brow UV, Material=@n1.Result) {materialFunction=/Game/Global/Materials/Character_Parent_2023/FlipbookFace_Parent_2023/MaterialFunctions/MF_FlipbookFace}
  n3 = MaterialFunctionCall!(Mask=Multiply(A=@n1.Tech Art Mask, B=OneMinus(None=@n2.FaceAlpha)), Fresnel=@n1.Fresnel, In=@n2.Material) {materialFunction=/Game/Global/MaterialFunctions/TechArt/MF_Advanced_CarClearCoat_v2}
  ClearCoatNormalCustomOutput!(None=@n3.ClearCoatNormal)
  MaterialFunctionCall(Material=@n3.Result, M=@n1.M, Pre Skinned Local Position=@n1.M, Distance Blend=@n1.M, Fresnel=@n1.M) {materialFunction=&MF_CharacterParent_2023_PostFX}
 ]
}
