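# Animation Blueprint for the SliceSprig male player body, driving jacket dynamics and button deformation.
# It copies the parent mesh pose, applies crouch and jump jacket offsets and a Control Rig, then caches it.
# RigidBody simulation adds skydiving forces, followed by skydive and right-thigh offsets, all bypassed above LOD 1 or when rigid bodies are disabled.
# Back-jacket bends, weighted bone constraints for buttons and cufflinks, and left-sleeve scaling finish the pose.
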AnimBlueprint /BRCosmetics/Characters/Player/Male/Medium/Bodies/M_MED_SliceSprig/Meshes/M_MED_SliceSprig_AnimBP
target_skeleton=&/M_MED_SliceSprig_Skeleton
graphs [
 AnimGraph {nodes [
   "Constraint - deform_cufflink_l (PPP)"!(ComponentPose, Alpha=1, ConstraintWeights_0=0.6954, ConstraintWeights_1:float=0.2641, ConstraintWeights_2:float=0.0406) -> (Pose) {
    alphaBoolBlend {blendOption=Linear}
    boneToModify {boneName=deform_cufflink_l}
    constraintSetup [{targetBone {boneName=deform_sleeve_l}}, {targetBone {boneName=lowerarm_twist_02_l}}, {targetBone {boneName=lowerarm_l}}]
    constraintWeights [1,1,1]
   }
   n1 = RigidBody(
    ComponentPose=Local To Component(LocalPose=Use cached pose 'pre_sim_cache'())
    ExternalForce=Break Fort Anim Input Skydiving External Force(FortAnimInput_SkydivingExternalForce=Get SkydivingExternalForceInput()) -> (
     bUseSkydivingVectorForce=~
     SkydivingVectorForce=~
     HeadToPelvisDirection=~
     FloatingMultiplier=~
     FloatingAdditive=~
     DivingMultiplier=~
     DivingAdditive=~
     ParachutingMultiplier=~
     ParachutingAdditive=~
    ).SkydivingVectorForce
   ) {
    bClampLinearTranslationLimitToRefPose=true
    bOverrideWorldGravity=true
    overridePhysicsAsset=&/M_MED_SliceSprig_Physics
    simulationSpace=BaseBoneSpace
    alphaBoolBlend {blendOption=Linear}
    baseBoneRef {boneName=pelvis}
    componentAppliedLinearAccClamp=<3000 3000 3000>
    componentLinearAccScale=<0.4 0.3 0.2>
    componentLinearVelScale=<0.4 0.3 0.2>
    overrideWorldGravity {z=-980}
   }
   n2 = "Transform (Modify) Bone - Bone: dyn_jacket_bk_01"(ComponentPose=@n1, Alpha=Get skydiving alpha() -> (skydiving alpha:double), Translation="0,0,0", Rotation="0,0.000,12", Scale="1,1,1") {rotationMode=BMM_Additive, alphaBoolBlend {blendOption=Linear}, boneToModify {boneName=dyn_jacket_bk_01}, rotation {roll=12}}
   n3 = "Transform (Modify) Bone - Bone: dyn_jacket_bk_r_01"(ComponentPose=@n2, Alpha=Get skydiving alpha(), Translation="0,0,0", Rotation="0,0,12", Scale="1,1,1") {rotationMode=BMM_Additive, alphaBoolBlend {blendOption=Linear}, boneToModify {boneName=dyn_jacket_bk_r_01}, rotation {roll=12}}
   n4 = "Transform (Modify) Bone - Bone: dyn_jacket_bk_l_01"(ComponentPose=@n3, Alpha=Get skydiving alpha(), Translation="0,0,0", Rotation="0,0,12", Scale="1,1,1") {rotationMode=BMM_Additive, alphaBoolBlend {blendOption=Linear}, boneToModify {boneName=dyn_jacket_bk_l_01}, rotation {roll=12}}
   n5 = "Transform (Modify) Bone - Bone: dyn_jacket_r_MSTR"(ComponentPose=@n4, Alpha=Get skydiving alpha(), Translation="0,0,0", Rotation="1,0.000,0.000", Scale="1,1,1") {rotationMode=BMM_Additive, alphaBoolBlend {blendOption=Linear}, boneToModify {boneName=dyn_jacket_r_MSTR}, rotation {pitch=1}}
   n6 = "Transform (Modify) Bone - Bone: dyn_jacket_l_MSTR"(ComponentPose=@n5, Alpha=Get skydiving alpha(), Translation="0,0,0", Rotation="1,0.000,0", Scale="1,1,1") {rotationMode=BMM_Additive, alphaBoolBlend {blendOption=Linear}, boneToModify {boneName=dyn_jacket_l_MSTR}, rotation {pitch=1}}
   n7 = "Transform (Modify) Bone - Bone: dyn_jacket_fr_r_MSTR"(ComponentPose=@n6, Alpha=Get RightThighUpAlpha() -> (RightThighUpAlpha:float), Translation="0,0,0", Rotation="5,0,0", Scale="1,1,1") {
    rotationMode=BMM_Additive
    rotationSpace=BCS_BoneSpace
    translationMode=BMM_Additive
    translationSpace=BCS_BoneSpace
    alphaBoolBlend {blendOption=Linear}
    boneToModify {boneName=dyn_jacket_fr_r_MSTR}
    rotation {pitch=5}
   }
   n8 = "Transform (Modify) Bone - Bone: dyn_jacket_r_MSTR"(ComponentPose=@n7, Alpha=Get RightThighUpAlpha(), Translation="0,0,0", Rotation="5,0,0", Scale="1,1,1") {
    rotationMode=BMM_Additive
    rotationSpace=BCS_BoneSpace
    translationMode=BMM_Additive
    translationSpace=BCS_BoneSpace
    alphaBoolBlend {blendOption=Linear}
    boneToModify {boneName=dyn_jacket_r_MSTR}
    rotation {pitch=5}
   }
   n9 = Component To Local(ComponentPose=@n8)
   n10 = Use cached pose 'pre_sim_cache'()
   n11 = Get dynamics_disabled() -> (dynamics_disabled:bool)
   n12 = "Transform (Modify) Bone - Bone: dyn_jacket_bk_02"!(
    ComponentPose=Local To Component(LocalPose=Blend Poses by bool(BlendPose_0=@n10, BlendPose_1=@n9, bActiveValue=@n11) {blendType=Linear})
    bAlphaBoolEnabled:bool=True
    Translation="0,0,0"
    Rotation="-30,0,0"
    Scale="1,1,1"
   ) -> (Pose) {
    alphaInputType=Bool
    rotationMode=BMM_Additive
    rotationSpace=BCS_BoneSpace
    alphaBoolBlend {blendInTime=0.3, blendOption=CircularIn, blendOutTime=0.3}
    boneToModify {boneName=dyn_jacket_bk_02}
    rotation {pitch=-30}
   }
   n13 = "Transform (Modify) Bone - Bone: dyn_jacket_bk_r_02"!(ComponentPose=@n12, bAlphaBoolEnabled:bool=True, Translation="0,0,0", Rotation="-30,0,0", Scale="1,1,1") -> (Pose) {
    alphaInputType=Bool
    rotationMode=BMM_Additive
    rotationSpace=BCS_BoneSpace
    alphaBoolBlend {blendInTime=0.3, blendOption=CircularIn, blendOutTime=0.3}
    boneToModify {boneName=dyn_jacket_bk_r_02}
    rotation {pitch=-30}
   }
   n14 = "Transform (Modify) Bone - Bone: dyn_jacket_bk_l_02"!(ComponentPose=@n13, bAlphaBoolEnabled:bool=True, Translation="0,0,0", Rotation="-30,0,0", Scale="1,1,1") -> (Pose) {
    alphaInputType=Bool
    rotationMode=BMM_Additive
    rotationSpace=BCS_BoneSpace
    alphaBoolBlend {blendInTime=0.3, blendOption=CircularIn, blendOutTime=0.3}
    boneToModify {boneName=dyn_jacket_bk_l_02}
    rotation {pitch=-30}
   }
   n15 = "Constraint - deform_button_01 (PPPPPP)"!(
    ComponentPose=@n14
    Alpha=1
    ConstraintWeights_0=0.6927
    ConstraintWeights_1:float=0.1527
    ConstraintWeights_2:float=0.0648
    ConstraintWeights_3:float=0.0623
    ConstraintWeights_4:float=0.0187
    ConstraintWeights_5:float=0.0087
   ) -> (Pose) {
    alphaBoolBlend {blendOption=Linear}
    boneToModify {boneName=deform_button_01}
    constraintSetup [
     {targetBone {boneName=spine_05}}
     {targetBone {boneName=clavicle_r}}
     {targetBone {boneName=spine_04}}
     {targetBone {boneName=neck_01}}
     {targetBone {boneName=spine_03}}
     {targetBone {boneName=clavicle_l}}
    ]
    constraintWeights [1,1,1,1,1,1]
   }
   n16 = "Constraint - deform_button_02 (PPPPPP)"!(
    ComponentPose=@n15
    Alpha=1
    ConstraintWeights_0=0.3344
    ConstraintWeights_1:float=0.2957
    ConstraintWeights_2:float=0.2286
    ConstraintWeights_3:float=0.0973
    ConstraintWeights_4:float=0.0345
    ConstraintWeights_5:float=0.0094
   ) -> (Pose) {
    alphaBoolBlend {blendOption=Linear}
    boneToModify {boneName=deform_button_02}
    constraintSetup [
     {targetBone {boneName=spine_05}}
     {targetBone {boneName=spine_04}}
     {targetBone {boneName=clavicle_r}}
     {targetBone {boneName=spine_03}}
     {targetBone {boneName=clavicle_l}}
     {targetBone {boneName=spine_02}}
    ]
    constraintWeights [1,1,1,1,1,1]
   }
   n17 = "Constraint - deform_button_03 (PPPPPP)"!(
    ComponentPose=@n16
    Alpha=1
    ConstraintWeights_0=0.3598
    ConstraintWeights_1:float=0.2071
    ConstraintWeights_2:float=0.114
    ConstraintWeights_3:float=0.1134
    ConstraintWeights_4:float=0.1069
    ConstraintWeights_5:float=0.0988
   ) -> (Pose) {
    alphaBoolBlend {blendOption=Linear}
    boneToModify {boneName=deform_button_03}
    constraintSetup [
     {targetBone {boneName=spine_04}}
     {targetBone {boneName=spine_03}}
     {targetBone {boneName=clavicle_l}}
     {targetBone {boneName=spine_02}}
     {targetBone {boneName=clavicle_r}}
     {targetBone {boneName=spine_05}}
    ]
    constraintWeights [1,1,1,1,1,1]
   }
   n18 = "Constraint - deform_button_04 (PPPPP)"!(
    ComponentPose=@n17
    Alpha=1
    ConstraintWeights_0=0.3193
    ConstraintWeights_1:float=0.2949
    ConstraintWeights_2:float=0.2507
    ConstraintWeights_3:float=0.0765
    ConstraintWeights_4:float=0.0586
   ) -> (Pose) {
    alphaBoolBlend {blendOption=Linear}
    boneToModify {boneName=deform_button_04}
    constraintSetup [
     {targetBone {boneName=spine_03}}
     {targetBone {boneName=spine_04}}
     {targetBone {boneName=spine_02}}
     {targetBone {boneName=spine_01}}
     {targetBone {boneName=clavicle_l}}
    ]
    constraintWeights [1,1,1,1,1]
   }
   n19 = "Constraint - deform_button_05 (PPPPPP)"!(
    ComponentPose=@n18
    Alpha=1
    ConstraintWeights_0=0.2868
    ConstraintWeights_1:float=0.2598
    ConstraintWeights_2:float=0.1702
    ConstraintWeights_3:float=0.167
    ConstraintWeights_4:float=0.1026
    ConstraintWeights_5:float=0.0137
   ) -> (Pose) {
    alphaBoolBlend {blendOption=Linear}
    boneToModify {boneName=deform_button_05}
    constraintSetup [
     {targetBone {boneName=spine_02}}
     {targetBone {boneName=spine_03}}
     {targetBone {boneName=spine_01}}
     {targetBone {boneName=pelvis}}
     {targetBone {boneName=spine_04}}
     {targetBone {boneName=clavicle_r}}
    ]
    constraintWeights [1,1,1,1,1,1]
   }
   n20 = "Constraint - deform_button_06 (PPPP)"!(
    ComponentPose=@n19
    Alpha=1
    ConstraintWeights_0=0.5057
    ConstraintWeights_1:float=0.1894
    ConstraintWeights_2:float=0.1608
    ConstraintWeights_3:float=0.1442
   ) -> (Pose) {
    alphaBoolBlend {blendOption=Linear}
    boneToModify {boneName=deform_button_06}
    constraintSetup [{targetBone {boneName=pelvis}}, {targetBone {boneName=spine_02}}, {targetBone {boneName=spine_01}}, {targetBone {boneName=spine_03}}]
    constraintWeights [1,1,1,1]
   }
   n21 = "Constraint - deform_button_07 (PPPPPP)"!(
    ComponentPose=@n20
    Alpha=1
    ConstraintWeights_0=0.2781
    ConstraintWeights_1:float=0.2636
    ConstraintWeights_2:float=0.1673
    ConstraintWeights_3:float=0.1535
    ConstraintWeights_4:float=0.105
    ConstraintWeights_5:float=0.0325
   ) -> (Pose) {
    alphaBoolBlend {blendOption=Linear}
    boneToModify {boneName=deform_button_07}
    constraintSetup [
     {targetBone {boneName=spine_02}}
     {targetBone {boneName=spine_03}}
     {targetBone {boneName=spine_01}}
     {targetBone {boneName=pelvis}}
     {targetBone {boneName=spine_04}}
     {targetBone {boneName=clavicle_r}}
    ]
    constraintWeights [1,1,1,1,1,1]
   }
   n22 = "Constraint - deform_button_08 (PPPPPPPP)"!(
    ComponentPose=@n21
    Alpha=1
    ConstraintWeights_0=0.5251
    ConstraintWeights_1:float=0.1902
    ConstraintWeights_2:float=0.0913
    ConstraintWeights_3:float=0.0609
    ConstraintWeights_4:float=0.0553
    ConstraintWeights_5:float=0.0497
    ConstraintWeights_6:float=0.0215
    ConstraintWeights_7:float=0.0061
   ) -> (Pose) {
    alphaBoolBlend {blendOption=Linear}
    boneToModify {boneName=deform_button_08}
    constraintSetup [
     {targetBone {boneName=dyn_jacket_fr_r_01}}
     {targetBone {boneName=pelvis}}
     {targetBone {boneName=dyn_jacket_fr_r_02}}
     {targetBone {boneName=spine_02}}
     {targetBone {boneName=spine_01}}
     {targetBone {boneName=spine_03}}
     {targetBone {boneName=dyn_jacket_r_01}}
     {targetBone {boneName=spine_04}}
    ]
    constraintWeights [1,1,1,1,1,1,1,1]
   }
   n23 = "Constraint - deform_button_13 (PPP)"!(ComponentPose=@n22, Alpha=1, ConstraintWeights_0=0.5692, ConstraintWeights_1:float=0.2349, ConstraintWeights_2:float=0.1959) -> (Pose) {
    alphaBoolBlend {blendOption=Linear}
    boneToModify {boneName=deform_button_13}
    constraintSetup [{targetBone {boneName=lowerarm_twist_02_r}}, {targetBone {boneName=deform_sleeve_r}}, {targetBone {boneName=lowerarm_r}}]
    constraintWeights [1,1,1]
   }
   n24 = "Constraint - deform_button_14 (PPP)"!(ComponentPose=@n23, Alpha=1, ConstraintWeights_0=0.5122, ConstraintWeights_1:float=0.3437, ConstraintWeights_2:float=0.1441) -> (Pose) {
    alphaBoolBlend {blendOption=Linear}
    boneToModify {boneName=deform_button_14}
    constraintSetup [{targetBone {boneName=lowerarm_twist_02_r}}, {targetBone {boneName=deform_sleeve_r}}, {targetBone {boneName=lowerarm_r}}]
    constraintWeights [1,1,1]
   }
   n25 = "Constraint - deform_button_15 (PPP)"!(ComponentPose=@n24, Alpha=1, ConstraintWeights_0=0.495, ConstraintWeights_1:float=0.4054, ConstraintWeights_2:float=0.0996) -> (Pose) {
    alphaBoolBlend {blendOption=Linear}
    boneToModify {boneName=deform_button_15}
    constraintSetup [{targetBone {boneName=deform_sleeve_r}}, {targetBone {boneName=lowerarm_twist_02_r}}, {targetBone {boneName=lowerarm_r}}]
    constraintWeights [1,1,1]
   }
   n26 = "Constraint - deform_button_16 (PPP)"!(ComponentPose=@n25, Alpha=1, ConstraintWeights_0=0.6203, ConstraintWeights_1:float=0.3163, ConstraintWeights_2:float=0.0634) -> (Pose) {
    alphaBoolBlend {blendOption=Linear}
    boneToModify {boneName=deform_button_16}
    constraintSetup [{targetBone {boneName=deform_sleeve_r}}, {targetBone {boneName=lowerarm_twist_02_r}}, {targetBone {boneName=lowerarm_r}}]
    constraintWeights [1,1,1]
   }
   n27 = "Constraint - deform_button_12 (PPP)"!(ComponentPose=@n26, Alpha=1, ConstraintWeights_0=0.5703, ConstraintWeights_1:float=0.2325, ConstraintWeights_2:float=0.1972) -> (Pose) {
    alphaBoolBlend {blendOption=Linear}
    boneToModify {boneName=deform_button_12}
    constraintSetup [{targetBone {boneName=lowerarm_twist_02_l}}, {targetBone {boneName=deform_sleeve_l}}, {targetBone {boneName=lowerarm_l}}]
    constraintWeights [1,1,1]
   }
   n28 = "Constraint - deform_button_11 (PPP)"!(ComponentPose=@n27, Alpha=1, ConstraintWeights_0=0.514, ConstraintWeights_1:float=0.3408, ConstraintWeights_2:float=0.1452) -> (Pose) {
    alphaBoolBlend {blendOption=Linear}
    boneToModify {boneName=deform_button_11}
    constraintSetup [{targetBone {boneName=lowerarm_twist_02_l}}, {targetBone {boneName=deform_sleeve_l}}, {targetBone {boneName=lowerarm_l}}]
    constraintWeights [1,1,1]
   }
   n29 = "Constraint - deform_button_10 (PPP)"!(ComponentPose=@n28, Alpha=1, ConstraintWeights_0=0.4922, ConstraintWeights_1:float=0.4072, ConstraintWeights_2:float=0.1006) -> (Pose) {
    alphaBoolBlend {blendOption=Linear}
    boneToModify {boneName=deform_button_10}
    constraintSetup [{targetBone {boneName=deform_sleeve_l}}, {targetBone {boneName=lowerarm_twist_02_l}}, {targetBone {boneName=lowerarm_l}}]
    constraintWeights [1,1,1]
   }
   n30 = "Constraint - deform_button_09 (PPP)"!(ComponentPose=@n29, Alpha=1, ConstraintWeights_0=0.6186, ConstraintWeights_1:float=0.3176, ConstraintWeights_2:float=0.0638) -> (Pose) {
    alphaBoolBlend {blendOption=Linear}
    boneToModify {boneName=deform_button_09}
    constraintSetup [{targetBone {boneName=deform_sleeve_l}}, {targetBone {boneName=lowerarm_twist_02_l}}, {targetBone {boneName=lowerarm_l}}]
    constraintWeights [1,1,1]
   }
   n31 = "Constraint - deform_cufflink_r (PPP)"!(ComponentPose=@n30, Alpha=1, ConstraintWeights_0=0.7283, ConstraintWeights_1:float=0.2304, ConstraintWeights_2:float=0.0413) -> (Pose) {
    alphaBoolBlend {blendOption=Linear}
    boneToModify {boneName=deform_cufflink_r}
    constraintSetup [{targetBone {boneName=deform_sleeve_r}}, {targetBone {boneName=lowerarm_twist_02_r}}, {targetBone {boneName=lowerarm_r}}]
    constraintWeights [1,1,1]
   }
   n32 = "Constraint - deform_cufflink_l (P)"!(ComponentPose=@n31, Alpha=1, ConstraintWeights_0=1, ConstraintWeights_1:float=0) -> (Pose) {
    alphaBoolBlend {blendOption=Linear}
    boneToModify {boneName=deform_cufflink_l}
    constraintSetup [{targetBone {boneName=deform_cuff_l}}]
    constraintWeights [1,0]
   }
   n33 = "Transform (Modify) Bone - Bone: deform_sleeve_l"(ComponentPose=@n32, Alpha=1, Translation="0,0,0", Rotation="0,0,0", Scale="1,1.1,1") {
    scaleMode=BMM_Replace
    scaleSpace=BCS_BoneSpace
    translationMode=BMM_Additive
    translationSpace=BCS_BoneSpace
    alphaBoolBlend {blendOption=Linear}
    boneToModify {boneName=deform_sleeve_l}
    scale {y=1.1}
   }
   Output Pose(Result=Component To Local(ComponentPose=@n33)) {layerGroup=None}
   n34 = "Transform (Modify) Bone - Bone: dyn_jacket_fr_l_MSTR"(
    ComponentPose=Local To Component(LocalPose=Copy Pose From Mesh() {bUseAttachedParent=true})
    Alpha=Get crouching alpha() -> (crouching alpha:double)
    Translation="-0.000,0.000,3"
    Rotation="52.974,-30.282,-57.791"
    Scale="1,1,1"
   ) {
    rotationMode=BMM_Additive
    rotationSpace=BCS_BoneSpace
    translationMode=BMM_Additive
    alphaBoolBlend {blendOption=Linear}
    boneToModify {boneName=dyn_jacket_fr_l_MSTR}
    rotation=<rot 52.9739 -30.2821 -57.7908>
    translation {z=3}
   }
   n35 = "Transform (Modify) Bone - Bone: dyn_jacket_fr_r_MSTR"(ComponentPose=@n34, Alpha=Get jump_alpha() -> (jump_alpha:double), Translation="0.000,0.000,4.488", Rotation="0,0,0", Scale="1,1,1") {translationMode=BMM_Additive, alphaBoolBlend {blendOption=Linear}, boneToModify {boneName=dyn_jacket_fr_r_MSTR}, translation {z=4.488}}
   n36 = Control Rig(Source=Component To Local(ComponentPose=@n35)) {
    alphaBoolBlend {blendOption=Linear}
    controlRigAssetReference {blueprintRigClass=&/M_MED_SliceSprig_CtrlRig.M_MED_SliceSprig_CtrlRig_C}
   }
   Save cached pose 'pre_sim_cache'(Pose=@n36) {cacheName=pre_sim_cache}
  ]}
 EventGraph {nodes [
   Event BlueprintUpdateAnimation()
   > n1 = Sequence!(execute) -> (then_0, then_1, then_2:Exec, then_3:Exec, then_4:Exec)
   > n2 = Branch(Condition=Get bIsJumping())
   Set jump_alpha(execute=[@n2.then, @n3.then], jump_alpha:double=1) -> (Output_Get:double)
   n3 = Branch(execute=@n2.else, Condition=Get bIsFalling())
   Set jump_alpha(execute=@n3.else, jump_alpha=0)
   n4 = Branch(execute=@n1.then_2, Condition=Get bIsSkyDiving())
   > Set skydiving alpha(skydiving alpha:double=1) -> (Output_Get:double)
   Set skydiving alpha(execute=@n4.else, skydiving alpha=0)
   Set RightThighUpAlpha(
    execute=@n1.then_3
    RightThighUpAlpha:float=GetAlphaFromReferenceTransformDelta(
     SkeletalMeshComponent=Get BaseMesh() -> (BaseMesh:Skeletal Mesh Component Object Reference)
     BoneName=thigh_r
     BaseName=pelvis
     DeltaComponentType=EulerY
     InputRange=(X=10,Y=25)
     TwistAxis="1,0,0"
    )
   ) -> (Output_Get:float)
   Set isWearingCape(execute=@n1.then_4, isWearingCape=CheckForBackbling().bIsBackblingACape)
   > Set isWearingCape_0(isWearingCape_0:bool=AND Boolean(A=Get isWearingCape() -> (isWearingCape:bool), B=Get bIsFalling())) -> (Output_Get:bool)
   n5 = Get CurrentLOD()
   n6 = Get bRigidBodyEnabled()
   Set dynamics_disabled(
    execute=@n1.then_1
    dynamics_disabled:bool=OR Boolean(A="integer > integer"(A=@n5, B=1) {is=title K2Node_CallFunction}, B=NOT Boolean(A=@n6))
   ) -> (Output_Get:bool)
   Event BlueprintInitializeAnimation()
   > Set BaseMesh(BaseMesh:Skeletal Mesh Component Object Reference=GetOwnerSkeletalMeshComponent()) -> (Output_Get:Skeletal Mesh Component Object Reference)
   Event OnEndCrouching()
   > Set crouching alpha(crouching alpha:double=0) -> (Output_Get:double)
   Event OnBeginCrouching()
   > Set crouching alpha(crouching alpha=1)
  ]}
]
variables [
 BaseMesh {type=SkeletalMeshComponent}
 crouching alpha {type=number}
 dynamics_disabled {type=boolean}
 isWearingCape {type=boolean}
 isWearingCape_0 {type=boolean}
 jump_alpha {type=number}
 RightThighUpAlpha {type=number}
 skydiving alpha {type=number}
]
